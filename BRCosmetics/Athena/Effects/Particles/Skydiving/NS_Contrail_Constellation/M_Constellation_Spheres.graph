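# Additive Niagara mesh-particle material for the Constellation skydiving contrail, drawing spheres or flare stars via an IsStar parameter.
# Spheres glow from tri-edge texture edge and gradient channels, while stars use particle color and a flare-texture mask.
# Sine waves over UV and time drive normal-based vertex wobble and a panning opacity mask shaped by dynamic parameters, Fresnel and camera fade.
# Low quality hides stars, removes their wobble and skips Fresnel.
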Material /BRCosmetics/Athena/Effects/Particles/Skydiving/NS_Contrail_Constellation/M_Constellation_Spheres
graph {
 properties {bUsedWithNiagaraMeshParticles=true, blendMode=BLEND_Additive, pixelDepthOffsetMode=PDOM_Legacy, usedShadingModels=Default Lit}
 nodes [
  n1 = TextureCoordinate()
  n2 = Add(A=ComponentMask(None=@n1) {r=true}, B=ComponentMask(None=Multiply(A=@n1) {constB=0.1}) {g=true})
  n3 = Time()
  n4 = VertexNormalWS()
  n5 = Multiply(A=Sine(None=Add(A=@n2, B=@n3)), B=@n4)
  n6 = ScalarParameter() {desc="1 for true, 0 for false", parameterName=IsStar}
  LinearInterpolate(A=@n5, B=QualitySwitch(Default=@n5, Low=@n4) {desc="Turns WPO off for stars on Low settings, when they are invisible"}, Alpha=@n6)
  n7 = TextureSample(UVs=TextureCoordinate() {coordinateIndex=1}) {texture=&/T_TriEdges}
  n8 = DynamicParameter() {defaultValue=<rgba -0.8 5 1 1>, paramNames [HexDensity,RandomStart,EdgeGlow,GradientGlow]}
  n9 = Add(
   A=Multiply(A=VectorParameter() {parameterName=Edge Color, defaultValue=<rgba 0.51 1.7751 10 1>}.RGB, B=Multiply(A=@n7.G, B=@n8.EdgeGlow))
   B=Multiply(A=Multiply(A=@n8.GradientGlow, B=@n7.B), B=VectorParameter() {parameterName=GradientColor, defaultValue=<rgba 0.0254 0.044 0.0208 1>}.RGB)
  )
  n10 = ParticleColor()
  MaterialFunctionCall(Emissive=Multiply(A=LinearInterpolate(A=@n9, Alpha=@n6), B=@n10.RGB)) {materialFunction=&MF_FX_EmissiveAdjustments}
  n11 = Fresnel(ExponentIn=Constant() {r=0.25})
  n12 = DynamicParameter() {defaultValue=<rgba -0.8 5 1 1>, paramNames [HexDensity,RandomStart,EdgeGlow,GradientGlow]}
  n13 = Add(A=@n12.RandomStart, B=Multiply(A=ScalarParameter() {defaultValue=1, parameterName=MaskPanSpeed}, B=Time()) {constA=0.8})
  n14 = Add(A=@n12.HexDensity, B=Sine(None=Add(A=@n13, B=@n2))) {constB=-0.4}
  n15 = Constant()
  n16 = Add(A=@n12.HexDensity)
  n17 = Constant() {r=1}
  n18 = Multiply(
   A=QualitySwitch(Default=OneMinus(None=@n11), Low=Constant() {r=1}) {desc=Turns Frsenel off on low settings}
   B=Saturate(None=MaterialFunctionCall(Input=@n14, Input Low=@n15, Input High=@n16, Target Low=@n15, Target High=@n17) {materialFunction=&RemapValueRange}.Result)
  )
  n19 = TextureSampleParameter2D(UVs=TextureCoordinate() {coordinateIndex=1}) {parameterName=FlareTexture, samplerType=SAMPLERTYPE_LinearGrayscale, texture=/Game/Environments/FrontEnd/Materials/T_LensFlareSimple}
  n20 = AppendVector(A=@n19.RGB, B=@n19.RGB)
  n21 = Multiply(
   A=QualitySwitch(Default=StaticComponentMaskParameter(None=@n20) {defaultG=true, parameterName=StarTextureMask}, Low=Constant()) {desc=Hides stars on low settings}
   B=@n18
  )
  Multiply(
   A=Multiply(A=@n10.A, B=LinearInterpolate(A=@n18, B=@n21, Alpha=@n6))
   B=MaterialFunctionCall(
    Fade Length=ScalarParameter() {defaultValue=150, parameterName=CameraFadeMax}
    Fade Offset=ScalarParameter() {defaultValue=25, parameterName=CameraFadeMin}
   ) {materialFunction=&CameraDepthFade}.Result
  )
 ]
}
