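# Constellation skydiving contrail cosmetic Niagara system, spawning from bones of the player's skeletal mesh.
# Four ribbons fade through color curves, joined by distance-culled sprite sparkles with a glow color and size.
# Rotating mesh spheres carry a fading warm stars color, and diving raises their spawn rate while shortening their lifetime.
# Front-end modifiers give sparkles and spheres extra upward velocity in menus, where the character is not falling.
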NiagaraSystem /BRCosmetics/Athena/Effects/Particles/Skydiving/NS_Contrail_Constellation/NS_Contrail_Constellation
dependencies {
 usedDataInterfaces [
  &NiagaraDataInterfaceCurve
  &NiagaraDataInterfaceVectorCurve
  &NiagaraDataInterfaceCamera
  &NiagaraDataInterfaceColorCurve
 ]
 usedDynamicInputs []
 usedModules [&SystemState]
}
emitters {
 FN_Contrails_Ribbon_v4_A {
  emitterUpdateScript [
   &EmitterScalability(
    Module.bApplySpawnCountScale=false
    Module.bApplySpawnCountScaleByDistanceFraction=false
    Module.bUseMaxDistance=false
    Module.bUseMinDistance=false
   ) {enabled=false}
   &EmitterLifeCycle(CompleteOnInactive=true) {enabled=false}
   &EmitterState()
   /Game/Effects/Niagara/Modules/Spawn/FN_SpawnContrails(All limbs=false, Global Spawnrate=60) {enabled=false}
   &FN_SpawnRibbon_v2(Number of Ribbons=4, Spawn Mode=::NewEnumerator0)
  ]
  particleSpawnScript [
   &FN_InitializeRibbon_v3(Enable Skeletal Position=true, Lifetime=0.7, Ribbon Width=60) {enabled=false}
   &FN_InitializeRibbon_v3(
    Color:LinearColor=/Niagara/DynamicInputs/TypeConversions/MakeLinearColorFromFloat(Value:NiagaraFloat=0.6)
    Enable Skeletal Position=true
    Lifetime=0.7
    Ribbon Width=50
    Sampled Position Coordinate Space:ENiagaraCoordinateSpace=::Local
    Sampled Rotation Coordinate Space:ENiagaraCoordinateSpace=::Local
    Sampled Velocity Coordinate Space:ENiagaraCoordinateSpace=::Local
   ) {module=FN_InitializeRibbon_v3001}
   &AddVelocity(Velocity=<0 0 410>)
   &AddVelocity(Velocity=&MakeCustomVectorFromBool(
     Bool=$Particles.Locked
     True Vector=<0 0 0>
     False Vector=&FN_FrontEndVector3Modifier(Gameplay=<0 0 500>, Front End=<0 0 500>)
    )) {module=AddVelocity001}
   &SampleSkeletalMeshSkeleton(Mesh:NiagaraDataInterfaceSkeletalMesh=$User.SkeletalMesh)
  ]
  particleUpdateScript [
   /Game/Effects/Niagara/Modules/Skydiving/FN_DivingAttributes/FN_DivingAttributes() {enabled=false}
   &SolveForcesAndVelocity(Clamp Velocity=false, Limit Acceleration=false)
   &FN_UpdateRibbon(
    Enable Skeletal Position=true
    Initial Color:LinearColor=$Particles.Initial.Color
    Initial RibbonWidth:NiagaraFloat=$Particles.Initial.RibbonWidth
   )
   &ScaleColor(
    Scale Alpha=&FloatFromCurve(FloatCurve=DataInterface() {propertyValues {Curve {keys [
         {arriveTangent=-0.2453, interpMode=RCIM_Cubic, leaveTangent=-0.2453, tangentMode=RCTM_User, time=-0.0044, value=2}
         {arriveTangent=-1.934, interpMode=RCIM_Cubic, leaveTangent=-1.934, tangentMode=RCTM_User, time=0.2157, value=1}
         {arriveTangent=-0.0191, interpMode=RCIM_Cubic, leaveTangent=-0.0191, tangentMode=RCTM_User, time=1.0011, value=0.002}
        ]}}})
    Scale RGB=&VectorFromCurve(VectorCurve=DataInterface() {propertyValues {
       XCurve {keys [{arriveTangent=-28.0246, interpMode=RCIM_Cubic, leaveTangent=-28.0246, tangentMode=RCTM_User, time=0, value=3}, <at 0.1919 0.2563>, <at 1 0>]}
       YCurve {keys [<at 0 1>,<at 0.1 1>,<at 1 0>]}
       ZCurve {keys [
         {arriveTangent=-0.0387, interpMode=RCIM_Cubic, leaveTangent=-0.0387, tangentMode=RCTM_User, time=0, value=0}
         {arriveTangent=0.0474, interpMode=RCIM_Cubic, leaveTangent=0.0474, tangentMode=RCTM_User, time=0.1, value=0.1}
         <at 1 0>
        ]}
      }})
    ScaleA=true
    ScaleRGB=true
   ) {enabled=false}
   &ScaleColor(
    Scale Alpha=&FloatFromCurve(FloatCurve=DataInterface() {propertyValues {Curve {keys [
         {arriveTangent=-16.7667, interpMode=RCIM_Cubic, leaveTangent=-16.7667, tangentMode=RCTM_User, time=-0.0044, value=10}
         {arriveTangent=-7.9091, interpMode=RCIM_Cubic, leaveTangent=-7.9091, tangentMode=RCTM_User, time=0.1708, value=2.2332}
         {arriveTangent=-2.6227, interpMode=RCIM_Cubic, leaveTangent=-2.6227, tangentMode=RCTM_User, time=1.0011, value=0.002}
        ]}}})
    Scale RGB=&VectorFromCurve(VectorCurve=DataInterface() {propertyValues {
       XCurve {keys [
         {arriveTangent=0.0247, interpMode=RCIM_Cubic, leaveTangent=0.0247, tangentMode=RCTM_User, time=0.01, value=1}
         {arriveTangent=-3.7866, interpMode=RCIM_Cubic, leaveTangent=-3.7866, tangentMode=RCTM_User, time=0.0927, value=0.3653}
         {arriveTangent=-0.5883, interpMode=RCIM_Cubic, leaveTangent=-0.5883, tangentMode=RCTM_User, time=0.1668, value=0.1313}
         {arriveTangent=-0.231, interpMode=RCIM_Cubic, leaveTangent=-0.231, tangentMode=RCTM_User, time=1, value=0}
        ]}
       YCurve {keys [
         {arriveTangent=0.5628, interpMode=RCIM_Cubic, leaveTangent=0.5628, tangentMode=RCTM_User, time=0.01, value=0.2}
         {arriveTangent=0.2538, interpMode=RCIM_Cubic, leaveTangent=0.2538, tangentMode=RCTM_User, time=0.1, value=0.5}
         <at 1 0>
        ]}
       ZCurve {keys [
         {arriveTangent=-0.0387, interpMode=RCIM_Cubic, leaveTangent=-0.0387, tangentMode=RCTM_User, time=0.01, value=0}
         {arriveTangent=-0.0252, interpMode=RCIM_Cubic, leaveTangent=-0.0252, tangentMode=RCTM_User, time=0.0678, value=0}
         {arriveTangent=0.0474, interpMode=RCIM_Cubic, leaveTangent=0.0474, tangentMode=RCTM_User, time=0.1, value=0.1}
         <at 1 0>
        ]}
      }})
    ScaleA=true
    ScaleRGB=true
   ) {module=ScaleColor001}
   &Material(Base Texture Time:NiagaraFloat=&Multiply_Float(A=$Emitter.Age, B=-1), PostMult=0.5455, Power=0.4408)
   &Color(
    Color=<rgba 1 1 1 0.8>
    Scale Alpha=&FloatFromCurve(
     FloatCurve=DataInterface() {propertyValues {Curve {keys [
         {arriveTangent=2.7939, interpMode=RCIM_Cubic, leaveTangent=2.7939, tangentMode=RCTM_User, time=-0.0046, value=1.0113}
         {arriveTangent=-1.0036, interpMode=RCIM_Cubic, leaveTangent=-4.3444, tangentMode=RCTM_Break, time=0.1387, value=1.232}
         <at 0.6 0.002>
        ]}}}
     Scale Curve=0.8
    )
   ) {enabled=false}
  ]
  rendererClasses [&NiagaraRibbonRendererProperties]
  renderers [{rendererClass=&NiagaraRibbonRendererProperties}]
 }
 FN_Contrails_Sparkles {
  emitterSpawnScript [&FN_PreserveSkydiveUserVariables() {module=PreserveUserVariables}]
  emitterUpdateScript [
   &EmitterState(Enable Distance Culling=true, MaxDistance=700, Module.bUseMaxDistance=true, Module.bUseMinDistance=true, Scalability Mode=::NewEnumerator1)
   &SpawnRate(SpawnRate=&UniformRangedFloat(Minimum=25, Maximum=50), Use Spawn Probability=false)
  ]
  particleSpawnScript [
   /Game/Effects/Niagara/Modules/Groups/FN_SetSpawnGroup_v1/FN_SetSpawnGroup(Emitter Bone Count=4)
   /Niagara/Modules/Spawn/Initialization/InitializeParticle(
    Color=<rgba 2 2 2 1>
    Lifetime=&UniformRangedFloat(Minimum=0.25, Maximum=1)
    Module.Write Color=true
    Module.Write Lifetime=true
    Module.Write Mass=true
    Module.Write Material Random=true
    Module.Write Position=false
    Module.Write Scale=false
    Module.Write SpriteRotation=false
    Module.Write SpriteSize=true
    Position=&SimulationPosition()
    Sprite Size=&UniformRangedVector2D(Minimum=<8 8>, Maximum=<10 10>)
   )
   &SampleSkeletalMeshSkeleton(Mesh:NiagaraDataInterfaceSkeletalMesh=$User.SkeletalMesh, Specific Bone Index:NiagaraInt32=$Particles.SpawnGroup)
   /Niagara/Modules/Spawn/Location/SkeletalMeshSkeletonLocation(Module.Inherit Velocity=false, Module.Use Sampled Bone as Ribbon ID=true)
   /Niagara/Modules/Spawn/Orientation/V2/InitialMeshOrientation(
    Orientation Vector=&UniformRangedVector(Minimum=<-1 -1 -1>, Maximum=<1 1 1>)
    Use Orientation Vector=true
    Use Rotation Vector=false
   )
   /SetVariables(Particles.RotationOrientation:NiagaraFloat=&UniformRangedFloat(Minimum=-1, Maximum=1)) {isSetParameters=true, module=SetVariables_#1}
   /SetVariables(Particles.RotationSpeed:Vector3f=&UniformRangedVector(Minimum=<-1 -1 -1>, Maximum=<1 1 1>)) {isSetParameters=true, module=SetVariables_#2}
   /Niagara/Modules/Spawn/Location/SphereLocation(MaskSpawnGroup=false, Module.Override Randomness=false, Sphere Radius=30, Surface Only=false)
  ]
  particleUpdateScript [
   &ParticleState()
   &InheritVelocity(
    Module.Limit Speed=false
    Velocity Scale=&UniformRangedVector(Minimum=<-0.0025 -0.0025 -0.005>, Maximum=<-0.005 -0.005 -0.02>)
   )
   &self:DistanceScaleValue(DistanceMax=200, DistanceMin=75) {enabled=false, module=DistanceScaleValue}
   &AddVelocity(
    Scale Added Velocity=<1 1 10>
    Velocity=&FN_FrontEndVector3Modifier(
     Gameplay=&UniformRangedVector(Minimum=<-1 -1 0>, Maximum=<1 1 0>)
     Front End=&UniformRangedVector(Minimum=<-1 -1 2>, Maximum=<1 1 8>)
    )
   )
   &SolveForcesAndVelocity(Clamp Velocity=false, Limit Acceleration=false)
   &ScaleSpriteSize(Scale Factor=&Multiply_Vector2(
     A=&Vector2DFromFloat(Value=$Emitter.Scalability.DistanceFraction)
     B=&Vector2DFromFloat(Value=&FloatFromCurve(FloatCurve=DataInterface() {propertyValues {Curve {keys [
           {arriveTangent=3.8361, interpMode=RCIM_Cubic, leaveTangent=3.8361, tangentMode=RCTM_User, time=0, value=0}
           {arriveTangent=0.002, interpMode=RCIM_Cubic, leaveTangent=0.002, tangentMode=RCTM_User, time=0.6, value=2}
           <at 1.0011 0.002>
          ]}}}))
    ))
   /SetVariables(Particles.Color=&MultiplyLinearColor(
     Linear Color=&MultiplyLinearColor(Linear Color=$Particles.Color, Scale Factor=&MakeVector4_FromFloat(Float=$Particles.DistanceValue))
     Scale Factor=&MakeVector4(X=1, Y=1, Z=1, W=0)
    )) {enabled=false, isSetParameters=true}
   /SetVariables(Particles.MultColor:LinearColor=&MakeLinearColorFromVectorAndFloat(
     "Vector (RGB)"=&Multiply_Vector(
      A=&MakeVectorFromLinearColorRGB(Color=$Particles.Color)
      B=&MakeVectorFromLinearColorRGB(Color=&ColorFromCurve(ColorCurve=DataInterface() {propertyValues {
          AlphaCurve {keys [<at -0.0023 1 RCIM_Cubic>,<at 0.9977 1 RCIM_Cubic>]}
          BlueCurve {keys [<at 0.1285 0.117 RCIM_Cubic>,<at 0.3093 0.086>,<at 0.5661 0.7237>,<at 0.7637 0.057>]}
          GreenCurve {keys [<at 0.1285 0.432 RCIM_Cubic>,<at 0.3093 1>,<at 0.5661 1>,<at 0.7637 1>]}
          RedCurve {keys [<at 0.1285 0.0082 RCIM_Cubic>,<at 0.3093 0.1622>,<at 0.5661 0.025>,<at 0.7637 0.6857>]}
         }}))
     )
     "Float (Alpha)"=1
    )) {isSetParameters=true}
   /SetVariables(Particles.GlowColor=&MultiplyLinearColor(
     Linear Color=<rgba 0.3 0.4 0.3 1>
     Scale Factor=&MakeVector4_FromLinearColor(LinearColor=$Particles.MultColor)
    )) {isSetParameters=true}
   /SetVariables(Particles.GlowSize=&Multiply_Vector2(A=$Particles.SpriteSize, B=<2 2>)) {isSetParameters=true}
  ]
  rendererClasses [&NiagaraSpriteRendererProperties]
  renderers [
   {rendererClass=&NiagaraSpriteRendererProperties}
   {rendererClass=&NiagaraSpriteRendererProperties, rendererIndex=1}
  ]
 }
 FN_Contrails_Spheres {
  emitterSpawnScript [&FN_PreserveSkydiveUserVariables() {module=PreserveUserVariables}]
  emitterUpdateScript [
   &EmitterState()
   /Game/Effects/Niagara/Modules/Skydiving/FN_DivingAttributes/FN_EmitterDivingAttributes_v1(Accumulate Dive=1.25, Accumulate Rest=1, Bone Index=5, SkeletalMesh:NiagaraDataInterfaceSkeletalMesh=$User.SkeletalMesh)
   &SpawnRate(
    SpawnRate=&Add_Float(
     A=&UniformRangedFloat(Minimum=40, Maximum=60)
     B=&Multiply_Float(A=60, B=$Emitter.Dive)
    )
    Use Spawn Probability=false
   )
  ]
  particleSpawnScript [
   &FN_SetSpawnGroup(Emitter Bone Count=2)
   &InitializeParticle(
    Color=&MakeLinearColor(
     R:NiagaraFloat=&UniformRangedFloat(Minimum=0, Maximum=1)
     G:NiagaraFloat=&UniformRangedFloat(Minimum=1, Maximum=20)
     B:NiagaraFloat=&UniformRangedFloat(Minimum=0.5, Maximum=4)
     A:NiagaraFloat=&UniformRangedFloat(Minimum=1, Maximum=5)
    )
    Lifetime=&Multiply_Float(
     A=&UniformRangedFloat(Minimum=0.25, Maximum=1)
     B=&OneMinusFloat(Float=&Multiply_Float(A=0.5, B=$Emitter.Dive))
    )
    Mesh Scale=&VectorFromFloat(Value=&UniformRangedFloat(Minimum=1.75, Maximum=3.5))
    Module.Write Color=true
    Module.Write Lifetime=true
    Module.Write Mass=true
    Module.Write Material Random=true
    Module.Write Position=false
    Module.Write Scale=true
    Module.Write SpriteRotation=false
    Module.Write SpriteSize=false
    Position=&SimulationPosition()
   )
   &SampleSkeletalMeshSkeleton(Mesh:NiagaraDataInterfaceSkeletalMesh=$User.SkeletalMesh, Specific Bone Index:NiagaraInt32=$Particles.SpawnGroup)
   &SkeletalMeshSkeletonLocation(Module.Inherit Velocity=false, Module.Use Sampled Bone as Ribbon ID=true)
   &InitialMeshOrientation(
    Orientation Vector=&UniformRangedVector(Minimum=<-1 -1 -1>, Maximum=<1 1 1>)
    Use Orientation Vector=true
    Use Rotation Vector=false
   )
   /SetVariables(Particles.RotationOrientation:NiagaraFloat=&UniformRangedFloat(Minimum=-1, Maximum=1)) {isSetParameters=true, module=SetVariables_#1}
   /SetVariables(Particles.RotationSpeed:Vector3f=&UniformRangedVector(Minimum=<-1 -1 -1>, Maximum=<1 1 1>)) {isSetParameters=true, module=SetVariables_#2}
   &SphereLocation(MaskSpawnGroup=false, Module.Override Randomness=false, Sphere Radius=20, Surface Only=false)
   &AddRotationalVelocity(Rotation Rate=&UniformRangedVector(Minimum=<-0.2 -0.2 -0.2>, Maximum=<0.2 0.2 0.2>))
   &InheritVelocity(
    Module.Limit Speed=false
    Velocity Scale=&UniformRangedVector(Minimum=<-0.1 -0.1 -0.2>, Maximum=<-0.1 -0.1 -0.5>)
   )
   &AddVelocity(
    Scale Added Velocity=<1 1 10>
    Velocity=&FN_FrontEndVector3Modifier(
     Gameplay=&UniformRangedVector(Minimum=<-1 -1 0>, Maximum=<1 1 0>)
     Front End=&UniformRangedVector(Minimum=<-1 -1 40>, Maximum=<1 1 60>)
    )
   )
   /SetVariables(Particles.DynamicMaterialParameter:Vector4f=&MakeVector4(
     X=&FloatFromCurve(FloatCurve=DataInterface() {propertyValues {Curve {keys [<at 0 -0.3>,<at 1 -0.8>]}}})
     Y=&UniformRangedFloat(Minimum=0, Maximum=1)
     Z=1
     W=1
    )) {isSetParameters=true}
  ]
  particleUpdateScript [
   &ParticleState()
   &SolveForcesAndVelocity(Clamp Velocity=false, Limit Acceleration=false)
   &ScaleMeshSize(Scale Factor=&VectorFromFloat(Value=&FloatFromCurve(FloatCurve=DataInterface() {propertyValues {Curve {keys [
          {arriveTangent=4.1911, interpMode=RCIM_Cubic, leaveTangent=4.1911, tangentMode=RCTM_User, time=0, value=0}
          {arriveTangent=1.0605, interpMode=RCIM_Cubic, leaveTangent=1.0605, tangentMode=RCTM_User, time=0.2141, value=0.75}
          {arriveTangent=0.1814, interpMode=RCIM_Cubic, leaveTangent=0.1814, tangentMode=RCTM_User, time=1, value=1}
         ]}}})))
   &SolveRotationalForcesAndVelocity()
   /SetVariables(Particles.StarsColor:LinearColor=&MultiplyLinearColor(
     Linear Color=&UniformRangedLinearColor(Minimum=<rgba 10 7.2941 4.1961 1>, Maximum=<rgba 10 10 4.1961 1>)
     Scale Factor=&MakeVector4_FromFloat(Float=&FloatFromCurve(FloatCurve=DataInterface() {propertyValues {Curve {keys [
           {arriveTangent=4.6845, interpMode=RCIM_Cubic, leaveTangent=4.6845, tangentMode=RCTM_User, time=0, value=0}
           {arriveTangent=0.0508, interpMode=RCIM_Cubic, leaveTangent=0.0508, tangentMode=RCTM_User, time=0.25, value=1}
           {arriveTangent=-0.4509, interpMode=RCIM_Cubic, leaveTangent=-0.4509, tangentMode=RCTM_User, time=1, value=0}
          ]}}}))
    )) {isSetParameters=true}
   &ScaleColor(
    Scale Alpha=&FloatFromCurve(FloatCurve=DataInterface() {propertyValues {Curve {keys [
         {arriveTangent=4.6845, interpMode=RCIM_Cubic, leaveTangent=4.6845, tangentMode=RCTM_User, time=0, value=0}
         {arriveTangent=-3.7579, interpMode=RCIM_Cubic, leaveTangent=-3.7579, tangentMode=RCTM_User, time=0.25, value=1}
         {arriveTangent=-0.4509, interpMode=RCIM_Cubic, leaveTangent=-0.4509, tangentMode=RCTM_User, time=1, value=0}
        ]}}})
    ScaleA=true
    ScaleRGB=true
   )
  ]
  rendererClasses [&NiagaraMeshRendererProperties]
  renderers [
   {rendererClass=&NiagaraMeshRendererProperties, rendererIndex=0}
   {rendererClass=&NiagaraMeshRendererProperties, rendererIndex=1}
  ]
 }
}
summary {
 systemName=NS_Contrail_Constellation
 emitters [
  {emitterName=FN_Contrails_Ribbon_v4_A, rendererClasses [&NiagaraRibbonRendererProperties]}
  {emitterName=FN_Contrails_Sparkles, rendererClasses [&NiagaraSpriteRendererProperties]}
  {emitterName=FN_Contrails_Spheres, rendererClasses [&NiagaraMeshRendererProperties]}
 ]
 userVariables [
  User.bIsFrontEnd {type=NiagaraBool}
  User.bIsFrontEndPreview {type=NiagaraBool}
  User.PlayerSpeed {type=NiagaraFloat, defaultValue=0, description=""}
  User.RibbonIDOffset {type=NiagaraInt32}
  User.RibbonSpawnRate {type=NiagaraFloat, defaultValue=40, description=""}
  User.SkeletalMesh {
   type=NiagaraDataInterfaceSkeletalMesh
   defaultValue {
    struct=NiagaraExt_VariableValue_DataInterface
    value {dataInterface=&self:NiagaraDataInterfaceSkeletalMesh_0}
   }
  }
  User.TrailAlpha {type=NiagaraFloat, defaultValue=1, description=""}
 ]
}
